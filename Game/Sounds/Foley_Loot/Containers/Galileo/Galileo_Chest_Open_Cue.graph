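# Sound cue for the opening of a Galileo loot chest, under container loot foley.
# A single wave player feeds the chest-open wave straight to the output, with no randomization or mixing nodes.
# It uses the shared item-interact attenuation and the loot reward sound class, overrides concurrency, and slightly boosts volume.
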SoundCue /Game/Sounds/Foley_Loot/Containers/Galileo/Galileo_Chest_Open_Cue
root=WavePlayer_2
nodes [{id=WavePlayer_2, properties {soundWaveAssetPtr=&/Galileo_Chest_Open}}]
properties {
 attenuationSettings=/Game/Sounds/Attenuation/Fort_Item_Interact_Att
 bOverrideConcurrency=true
 maxDistance=1850
 soundClassObject=&Fort_Loot_Reward_PSM
 volumeMultiplier=1.05
}
