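# Sound cue for the long searching sound of the Galileo chest loot container, built from Treasure Chest search sounds.
# A mixer layers a start one-shot with one of three equally weighted looping search variations chosen at random, each layer passing through its own pitch modulator.
# It uses item-interact mono attenuation and the loot-collect sound class, with a concurrency override that prevents new instances once the limit is reached.
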SoundCue /Game/Sounds/Foley_Loot/Containers/Galileo/Galileo_Chest_Searching_Long_Cue
root=Mixer_1
nodes [
 {id=Mixer_1, children [Modulator_2,Modulator_3], properties {inputVolume [1,1]}}
 {id=Modulator_2, children [WavePlayer_5], properties {pitchMax=1.2, pitchMin=1, volumeMax=1, volumeMin=1}}
 {id=Modulator_3, children [Random_0], properties {pitchMax=1.1, volumeMax=1, volumeMin=1}}
 {id=Random_0, children [WavePlayer_6,WavePlayer_7,WavePlayer_8], properties {bShouldExcludeFromBranchCulling=true, weights [1,1,1]}}
 {id=WavePlayer_5, properties {soundWaveAssetPtr=/Game/Sounds/Foley_Loot/Containers/Treasure_Chest/Treasure_Chest_Search_Start_01}}
 {
  id=WavePlayer_6
  properties {bLooping=true, soundWaveAssetPtr=/Game/Sounds/Foley_Loot/Containers/Treasure_Chest/Treasure_Chest_Search_Loop_01}
 }
 {
  id=WavePlayer_7
  properties {bLooping=true, soundWaveAssetPtr=/Game/Sounds/Foley_Loot/Containers/Treasure_Chest/Treasure_Chest_Search_Loop_02}
 }
 {
  id=WavePlayer_8
  properties {bLooping=true, soundWaveAssetPtr=/Game/Sounds/Foley_Loot/Containers/Treasure_Chest/Treasure_Chest_Search_Loop_03}
 }
]
properties {
 attenuationSettings=/Game/Sounds/Attenuation/Fort_Item_Interact_Mono_Att
 bOverrideConcurrency=true
 maxDistance=1500
 soundClassObject=&Fort_Loot_Collect
 volumeMultiplier=1.7657
 concurrencyOverrides {resolutionRule=PreventNew}
}
